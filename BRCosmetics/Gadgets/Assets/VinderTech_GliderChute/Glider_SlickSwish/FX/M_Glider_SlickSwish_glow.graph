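# Translucent, unlit, two-sided glow material for Niagara sprite and mesh particles in the SlickSwish glider's FX.
# A channel-selectable mask, modulated by panning noise and a power, drives two color gradients blended by dynamic Variant_ColorSwitch.
# Particle color, emissive strength and an optional panning color texture shape the emissive.
# Opacity combines mask, particle alpha, dynamic sphere mask and view-angle fade, with depth fade only on default and deferred paths.
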Material /BRCosmetics/Gadgets/Assets/VinderTech_GliderChute/Glider_SlickSwish/FX/M_Glider_SlickSwish_glow
graph {
 properties {
  bAutomaticallySetUsageInEditor=false
  bUsedWithNiagaraMeshParticles=true
  bUsedWithNiagaraSprites=true
  blendMode=BLEND_Translucent
  shadingModel=MSM_Unlit
  twoSided=true
  usedShadingModels=Unlit
 }
 nodes [
  Constant() {r=1}
  n1 = TextureSampleParameter2D() {parameterName=MaskTexture, texture=&/T_Glider_SlickSwish_GlowRGB}
  n2 = LinearInterpolate(A=@n1.R, B=@n1.R, Alpha=ScalarParameter() {parameterName=Use_G})
  n3 = ScalarParameter() {parameterName=Use_B}
  n4 = TextureSample(UVs=Panner(Coordinate=TextureCoordinate()) {speedY=0.15}) {samplerType=SAMPLERTYPE_LinearGrayscale, texture=/Engine/Functions/Engine_MaterialFunctions02/ExampleContent/Textures/LowResBlurredNoise}
  n5 = Power(
   Base=Multiply(A=LinearInterpolate(A=@n2, B=@n1.B, Alpha=@n3), B=LinearInterpolate(Alpha=@n4.R) {constA=0.3})
   Exp=ScalarParameter() {defaultValue=1, parameterName=Maskpower}
  )
  n6 = LinearInterpolate(
   A=VectorParameter() {parameterName=Color A, defaultValue=<rgba 1 0.0129 0.2623 1>}.RGB
   B=VectorParameter() {parameterName=Color B, defaultValue=<rgba 1 0.1875 0.5328 1>}.RGB
   Alpha=@n5
  )
  n7 = LinearInterpolate(
   A=VectorParameter() {parameterName=Variant_Color A, defaultValue=<rgba 0.3419 0.0129 1 1>}.RGB
   B=VectorParameter() {parameterName=Variant_Color B, defaultValue=<rgba 0.526 0.1875 1 1>}.RGB
   Alpha=@n5
  )
  n8 = DynamicParameter() {defaultValue=<rgba 1 1 0 1>, paramNames [FE_MeshFade, SphereMask, Variant_ColorSwitch, Param4]}
  n9 = ParticleColor()
  n10 = Multiply(
   A=Multiply(A=LinearInterpolate(A=@n6, B=@n7, Alpha=@n8.Variant_ColorSwitch), B=ScalarParameter() {defaultValue=2, parameterName=Emissive Mult})
   B=@n9.RGB
  )
  n11 = Multiply(A=TextureCoordinate(), B=ScalarParameter() {defaultValue=1, parameterName=ColorTexture_Tiling})
  n12 = Desaturation(
   None=TextureSampleParameter2D(UVs=Panner(Coordinate=@n11) {speedX=-0.2, speedY=0.2}) {parameterName=ColorTexture, samplerType=SAMPLERTYPE_LinearColor, texture=&HighContrastNoise}.RGB
   Fraction=ScalarParameter() {parameterName=ColorTexture_Desaturation}
  )
  MaterialFunctionCall(Emissive=LinearInterpolate(A=@n10, B=Multiply(A=@n12, B=@n10), Alpha=ScalarParameter() {parameterName=Use_ColorTexture})) {materialFunction=&MF_FX_EmissiveAdjustments}
  n13 = DynamicParameter() {defaultValue=<rgba 1 1 0 1>, paramNames [FE_MeshFade, SphereMask, Variant_ColorSwitch, Param4]}
  n14 = OneMinus(None=SphereMask(A=TextureCoordinate() {vTiling=0.5}, B=Constant2Vector() {g=0.45, r=0.5}, Radius=Constant() {r=0.1}, Hardness=Constant()))
  n15 = ScalarParameter() {defaultValue=2, parameterName=MeshFade}
  n16 = DynamicParameter() {defaultValue=<rgba 1 1 0 1>, paramNames [FE_MeshFade, SphereMask, Variant_ColorSwitch, Param4]}
  n17 = Multiply(
   A=Multiply(A=Multiply(A=@n9.A, B=@n5), B=LinearInterpolate(A=@n13.SphereMask, Alpha=@n14) {constA=1})
   B=MaterialFunctionCall("FadeSensitivity (optional)"=Multiply(A=@n15, B=@n16.FE_MeshFade)) {materialFunction=&FadeBasedOnViewAngle}.Result
  )
  n18 = DepthFade(Opacity=@n17, FadeDistance=ScalarParameter() {defaultValue=10, parameterName=DepthFade})
  ShadingPathSwitch(Default=@n18, Deferred=@n18, Forward=@n17, Mobile=@n17)
 ]
}
