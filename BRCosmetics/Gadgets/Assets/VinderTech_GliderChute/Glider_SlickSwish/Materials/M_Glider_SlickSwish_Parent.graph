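# Parent material for the SlickSwish glider, adding an optional animated emissive layer to the 2023 character parent material.
# The layer sits between the PreFX and PostFX functions, with mask-channel blending and optional base color zeroing.
# A panning, normal-map-distorted gradient uses single-plane, triplanar or UV mapping via static switches.
# The gradient is colored by curve atlas, gradient texture or tint colors, with an invertible Fresnel-style falloff shaping the emissive.
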Material /BRCosmetics/Gadgets/Assets/VinderTech_GliderChute/Glider_SlickSwish/Materials/M_Glider_SlickSwish_Parent
graph {
 properties {
  bUseMaterialAttributes=true
  bUsedWithMeshDeformer=true
  bUsedWithMorphTargets=true
  bUsedWithSkeletalMesh=true
  pixelDepthOffsetMode=PDOM_Legacy
  shadingModel=MSM_FromMaterialExpression
  usedShadingModels=Default Lit | Subsurface | Clear Coat | Subsurface Profile | Cloth
 }
 nodes [
  n1 = NamedRerouteUsage()
  n2 = GetMaterialAttributes(None=NamedRerouteUsage()) {attributeGetTypes [#1, #2]}
  n3 = NamedRerouteUsage()
  n4 = NamedRerouteUsage()
  n5 = OneMinus(None=@n4)
  n6 = ScalarParameter() {group=06_Advanced Emissive, parameterName=InvertAdvancedEmissiveFresnel, sortPriority=39}
  n7 = Saturate(None=Power(
    Base=LinearInterpolate(A=@n5, B=@n4, Alpha=@n6)
    Exp=ScalarParameter() {defaultValue=2, group=06_Advanced Emissive, parameterName=AdvancedEmissiveFresnelPower, sortPriority=39}
   ))
  n8 = SetMaterialAttributes!(
   MaterialAttributes=@n2.MaterialAttributes
   Base Color=StaticSwitchParameter(True=Constant(), False=@n2.Base Color) {defaultValue=true, group=06_Advanced Emissive, parameterName=SetBaseColorToZero}
   Emissive Color=Add(A=@n2.Emissive Color, B=Multiply(A=@n3, B=@n7))
  ) {attributeSetTypes [#2, #1]}
  n9 = ChannelMaskParameter(None=NamedRerouteUsage()) {group=05_Tech Art Masks, parameterName=AdvancedEmissiveMaskChannel}
  n10 = MaterialFunctionCall() {materialFunction=&MF_CharacterParent_2023_PreFX}
  MaterialFunctionCall(
   Material=StaticSwitchParameter(True=BlendMaterialAttributes(A=@n1, B=@n8, Alpha=@n9.R) {vertexAttributeBlendType=UseA}, False=NamedRerouteUsage()) {group=04_Tech Art, parameterName=Use_SlickSwish_AdvancedEmissive}
   M=@n10.M
   Pre Skinned Local Position=@n10.M
   Distance Blend=@n10.M
   Fresnel=@n10.M
  ) {materialFunction=&MF_CharacterParent_2023_PostFX}
  NamedRerouteDeclaration(None=@n10.Pre skinned Local Position) {name=InLocalPos, nodeColor=<rgba 0.6641 0 1 1>}
  NamedRerouteDeclaration(None=VertexInterpolator(VS=PreSkinnedNormal()).PS) {name=InLocalNormal, nodeColor=<rgba 1 0.0234 0 1>}
  NamedRerouteDeclaration(None=@n10.Result) {name=InMaterial, nodeColor=<rgba 0 0.2969 1 1>}
  NamedRerouteDeclaration(None=@n10.Fresnel) {name=InFresnel, nodeColor=<rgba 0.6172 1 0 1>}
  NamedRerouteDeclaration(None=@n10.Tech Art Mask) {name=TechArtMask, nodeColor=<rgba 0 0.5781 1 1>}
  NamedRerouteDeclaration(None=@n2.MaterialAttributes) {name=Material Attributes, nodeColor=<rgba 1 0.5156 0 1>}
  n11 = NamedRerouteUsage()
  n12 = ScalarParameter() {group=06_Advanced Emissive, parameterName=UVOffsetU}
  n13 = ScalarParameter() {group=06_Advanced Emissive, parameterName=UVOffsetV}
  n14 = StaticSwitchParameter(
   True=Multiply(A=ChannelMaskParameter(None=@n11) {group=06_Advanced Emissive, parameterName=PanningOffsetMaskChannel}.R, B=AppendVector(A=@n12, B=@n13))
   False=Constant()
  ) {group=06_Advanced Emissive, parameterName=OffsetPanningByTechArtMaskChannel}
  n15 = Multiply(
   A=NamedRerouteUsage()
   B=Divide(
    A=NamedRerouteUsage()
    B=ScalarParameter() {defaultValue=40, group=06_Advanced Emissive, parameterName=EmissiveTextureConstantScaleDivisor, sortPriority=4}
   )
  )
  n16 = Time()
  n17 = VectorParameter() {group=06_Advanced Emissive, parameterName=Emissive Panning Speed, sortPriority=2, defaultValue=<rgba 0 0.2 0.2 1>}
  n18 = Add(A=@n15, B=Frac(None=Multiply(A=@n16, B=@n17.RGB)))
  n19 = ComponentMask(None=@n18) {g=true, r=true}
  n20 = ScalarParameter() {group=06_Advanced Emissive Distortion, parameterName=EmissiveGradientDistortionAmt}
  n21 = Multiply(A=@n15, B=ScalarParameter() {defaultValue=1, group=06_Advanced Emissive Distortion, parameterName=EmissiveGradientDistortion_Scale})
  n22 = ComponentMask(None=@n21) {g=true, r=true}
  n23 = AppendVector(
   A=ScalarParameter() {group=06_Advanced Emissive Distortion, parameterName=EmissiveGradientDistortion_PanSpeed_U}
   B=ScalarParameter() {group=06_Advanced Emissive Distortion, parameterName=EmissiveGradientDistortion_PanSpeed_V}
  )
  n24 = VectorParameter() {group=06_Advanced Emissive Distortion, parameterName=EmissiveGradientDistortionPanningSpeed, sortPriority=2, defaultValue=<rgba 0 0.2 0.2 1>}
  n25 = Add(A=Frac(None=Multiply(A=@n16, B=@n24.RGB)), B=@n21)
  n26 = TextureSampleParameter2D(UVs=StaticSwitchParameter(True=Panner(Coordinate=@n22, Speed=@n23), False=ComponentMask(None=@n25) {g=true, r=true}) {defaultValue=true, group=06_Advanced Emissive, parameterName=UseZAxisOnlyForPanningTexture}) {
   group=06_Advanced Emissive Distortion
   parameterName=EmissiveGradientDistortion
   samplerSource=SSM_Wrap_WorldGroupSettings
   samplerType=SAMPLERTYPE_Normal
   texture=&FlatNormalBlue
  }
  n27 = ComponentMask(None=@n26.RGB) {g=true, r=true}
  n28 = TextureObjectParameter() {
   group=06_Advanced Emissive
   parameterName=EmissiveGradient
   samplerType=SAMPLERTYPE_Masks
   sortPriority=1
   texture=/Game/Global/Textures/Default/Masks_and_Gradients/T_VertGradient_MASK
  }
  n29 = TextureSample(UVs=Add(A=Add(A=@n14, B=@n19), B=Multiply(A=@n20, B=@n27)), Tex=@n28) {
   samplerSource=SSM_Wrap_WorldGroupSettings
   samplerType=SAMPLERTYPE_Masks
   texture=/Game/Weapons/WeaponSkins/Textures/T_Swirly_Noise_highContrast
  }
  n30 = ComponentMask(None=@n18) {b=true, g=true}
  n31 = TextureSampleParameter2D(UVs=ComponentMask(None=@n25) {b=true, g=true}) {
   group=06_Advanced Emissive Distortion
   parameterName=EmissiveGradientDistortion
   samplerSource=SSM_Wrap_WorldGroupSettings
   samplerType=SAMPLERTYPE_Normal
   texture=&FlatNormalBlue
  }
  n32 = ComponentMask(None=@n31.RGB) {g=true, r=true}
  n33 = TextureSample(UVs=Add(A=Add(A=@n14, B=@n30), B=Multiply(A=@n20, B=@n32)), Tex=@n28) {
   samplerSource=SSM_Wrap_WorldGroupSettings
   samplerType=SAMPLERTYPE_Masks
   texture=&T_Swirly_Noise_highContrast
  }
  n34 = ComponentMask(None=@n18) {b=true, r=true}
  n35 = TextureSampleParameter2D(UVs=ComponentMask(None=@n25) {b=true, r=true}) {
   group=06_Advanced Emissive Distortion
   parameterName=EmissiveGradientDistortion
   samplerSource=SSM_Wrap_WorldGroupSettings
   samplerType=SAMPLERTYPE_Normal
   texture=&FlatNormalBlue
  }
  n36 = ComponentMask(None=@n35.RGB) {g=true, r=true}
  n37 = TextureSample(UVs=Add(A=Add(A=@n14, B=@n34), B=Multiply(A=@n20, B=@n36)), Tex=@n28) {
   samplerSource=SSM_Wrap_WorldGroupSettings
   samplerType=SAMPLERTYPE_Masks
   texture=&T_Swirly_Noise_highContrast
  }
  n38 = VertexInterpolator(VS=PreSkinnedNormal())
  n39 = ComponentMask(None=@n38.PS) {g=true}
  n40 = ScalarParameter() {defaultValue=1, group=06_Advanced Emissive, parameterName=TriPlannerAxisBlendStrength, sortPriority=4}
  n41 = ComponentMask(None=MaterialFunctionCall(In=Abs(None=@n39), Contrast=@n40) {materialFunction=&CheapContrast}.Result) {g=true}
  n42 = MaterialFunctionCall(In=Abs(None=ComponentMask(None=@n38.PS) {b=true}), Contrast=@n40) {materialFunction=&CheapContrast}
  n43 = LinearInterpolate(
   A=LinearInterpolate(A=@n33.RGB, B=@n37.RGB, Alpha=Saturate(None=@n41))
   B=@n29.RGB
   Alpha=Saturate(None=ComponentMask(None=@n42.Result) {b=true})
  )
  n44 = TextureCoordinate() {coordinateIndex=1}
  n45 = TextureCoordinate()
  n46 = TextureSampleParameter2D() {
   group=06_Advanced Emissive
   parameterName=UVScaleMap
   samplerSource=SSM_Wrap_WorldGroupSettings
   samplerType=SAMPLERTYPE_Masks
   texture=&T_WhiteMask
  }
  n47 = Add(A=@n46.RGB, B=OneMinus(None=ScalarParameter() {defaultValue=1, group=06_Advanced Emissive, parameterName=UVScaleMapIntensity}))
  n48 = ScalarParameter() {defaultValue=1, group=06_Advanced Emissive, parameterName=EmissivePannerScale}
  n49 = StaticSwitchParameter(True=Multiply(A=Saturate(None=@n47), B=@n48), False=@n48) {
   desc="This allows modification of the panning EmissiveGradient texture direction using a RG \"UV\" texture.\r\nIf \"WorldAlignedGradient\" is true, it uses RGB."
   group=06_Advanced Emissive
   parameterName=UseUVScaleMap
  }
  n50 = ComponentMask(None=NamedRerouteDeclaration(None=@n49) {name=UV Scale, nodeColor=<rgba 0.6172 0 1 1>}) {
   desc="Kendall added this node on 5/27/2020.\r\nPreviously, this just piped the 3-vector value from\r\nthe texture sample straight into the UV's, which\r\ncaused the material to fail compile"
   g=true
   r=true
  }
  n51 = Add(
   A=@n14
   B=Multiply(A=StaticSwitchParameter(True=@n44, False=@n45) {group=06_Advanced Emissive, parameterName=UseSecondUVChannelForPanning}, B=@n50)
  )
  n52 = AppendVector(
   A=ScalarParameter() {group=06_Advanced Emissive, parameterName=GradientSpeed_X, sortPriority=2}
   B=ScalarParameter() {defaultValue=0.2, group=06_Advanced Emissive, parameterName=GradientSpeed_Y, sortPriority=2}
  )
  n53 = ScalarParameter() {group=06_Advanced Emissive Distortion, parameterName=EmissiveGradientDistortionAmt}
  n54 = Multiply(A=@n51, B=ScalarParameter() {defaultValue=1, group=06_Advanced Emissive Distortion, parameterName=EmissiveGradientDistortion_Scale})
  n55 = ScalarParameter() {group=06_Advanced Emissive Distortion, parameterName=EmissiveGradientDistortion_PanSpeed_U}
  n56 = ScalarParameter() {group=06_Advanced Emissive Distortion, parameterName=EmissiveGradientDistortion_PanSpeed_V}
  n57 = TextureSampleParameter2D(UVs=Panner(Coordinate=ComponentMask(None=@n54) {g=true, r=true}, Speed=AppendVector(A=@n55, B=@n56))) {
   group=06_Advanced Emissive Distortion
   parameterName=EmissiveGradientDistortion
   samplerSource=SSM_Wrap_WorldGroupSettings
   samplerType=SAMPLERTYPE_Normal
   texture=&FlatNormalBlue
  }
  n58 = ComponentMask(None=@n57.RGB) {g=true, r=true}
  n59 = TextureSample(UVs=Add(A=Panner(Coordinate=@n51, Speed=@n52), B=Multiply(A=@n53, B=@n58)), Tex=@n28) {
   samplerSource=SSM_Wrap_WorldGroupSettings
   samplerType=SAMPLERTYPE_Masks
   texture=&T_Swirly_Noise_highContrast
  }
  n60 = Power(
   Base=StaticSwitchParameter(
    True=StaticSwitchParameter(True=@n29.RGB, False=@n43) {defaultValue=true, group=06_Advanced Emissive, parameterName=UseZAxisOnlyForPanningTexture}
    False=@n59.RGB
   ) {defaultValue=true, group=06_Advanced Emissive, parameterName=UseWorldAlignedGradient, sortPriority=0}
   Exp=ScalarParameter() {defaultValue=1, group=06_Advanced Emissive Distortion, parameterName=PanningTextureValueExponent}
  )
  n61 = CurveAtlasRowParameter(CurveTime=@n60) {
   atlas=/Game/Characters/Shared/CurveAtlas/AthenaCharacterCurveAtlas
   curve=/Game/Characters/Player/Male/Medium/Bodies/M_MED_Cereal_Box/Textures/CerealBoxEyesColorCurveCosmos
   defaultValue=24
   group=06_Advanced Emissive
   parameterName=PanningEmissiveColorCurve
   sortPriority=6
  }
  n62 = TextureSampleParameter2D(UVs=ComponentMask(None=@n60) {g=true, r=true}) {
   group=06_Advanced Emissive
   parameterName=PanningEmissiveColorGradient
   samplerSource=SSM_Wrap_WorldGroupSettings
   sortPriority=6
   texture=/Game/Global/Textures/Default/Masks_and_Gradients/3pxRGBImage
  }
  n63 = VectorParameter() {group=06_Advanced Emissive, parameterName=AdvancedEmissiveColor, sortPriority=15, defaultValue=<rgba 1 1 1 1>}
  n64 = VectorParameter() {group=06_Advanced Emissive, parameterName=AdvancedEmissiveColor2, sortPriority=15, defaultValue=<rgba 1 1 1 1>}
  n65 = NamedRerouteUsage()
  n66 = StaticSwitchParameter(
   True=LinearInterpolate(
    A=@n63.RGB
    B=@n64.RGB
    Alpha=ChannelMaskParameter(None=@n65) {group=06_Advanced Emissive, parameterName=SecondEmissiveColorChannel, sortPriority=11}.R
   )
   False=@n63.RGB
  ) {group=06_Advanced Emissive, parameterName=UseTechArtMaskChannelForSecondEmissiveColor, sortPriority=11}
  n67 = LinearInterpolate(
   A=ScalarParameter() {group=06_Advanced Emissive, parameterName=AdvancedEmissiveBrightnessMin, sortPriority=16}
   B=ScalarParameter() {defaultValue=1, group=06_Advanced Emissive, parameterName=AdvancedEmissiveBrightnessMax, sortPriority=16}
   Alpha=@n60
  )
  NamedRerouteDeclaration(None=StaticSwitchParameter(
    True=StaticSwitchParameter(True=@n61, False=@n62.RGB) {group=06_Advanced Emissive, parameterName=UseCurveAtlasForGradient, sortPriority=0}
    False=Multiply(A=@n66, B=@n67)
   ) {group=06_Advanced Emissive, parameterName=UseGradientForColor, sortPriority=0}) {name=AdvancedEmissiveOutput, nodeColor=<rgba 1 0 0.2578 1>}
 ]
}
